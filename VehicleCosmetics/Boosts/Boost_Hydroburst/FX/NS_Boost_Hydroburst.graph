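# Niagara system for the Hydroburst vehicle boost cosmetic, emitting from filtered skeletal mesh sockets with a per-car offset.
# Two multi-renderer ribbon emitters form the exhaust: a jittered, width-pulsing streak and a trail with animated sine wobble and dissolve parameters.
# A third emitter repeatedly bursts short-lived, randomly oriented meshes and sprites at each port.
# Painted-variant overrides recolor all emitters, while game mode switches ribbon lifetimes and inherited source movement fades with age.
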NiagaraSystem /VehicleCosmetics/Boosts/Boost_Hydroburst/FX/NS_Boost_Hydroburst
dependencies {
 usedDataInterfaces [
  &NiagaraDataInterfaceColorCurve
  &NiagaraDataInterfaceCurve
  &NiagaraDataInterfaceVectorCurve
  &NiagaraDataInterfaceSkeletalMesh
 ]
 usedDynamicInputs []
 usedModules [
  &self:SystemSpawnScript.SystemScriptSource.SystemScriptGraph.NiagaraNodeAssignment_0.SetVariables_#1
  &SystemState
 ]
}
emitters {
 NE_VC_Boost_Ribbon_Parent003 {
  emitterUpdateScript [
   &EmitterState(
    Inactive Response=::NewEnumerator1
    Life Cycle Mode=::NewEnumerator1
    Loop Duration=&RandomRangeFloat(Minimum=0.2, Maximum=0.5)
    Recalculate Duration Each Loop=true
    UseLoopDelay=false
   )
   &VC_FilteredSocketCount()
   &VC_RateQualityHandler(High Spawn Rate=60, Low Spawn Rate=60, Medium Spawn Rate=60, Use Custom Rate?=true) {enabled=false}
   &VC_SpawnRibbon_v1(Number of Ribbons=$Emitter.PortCount, Spawn Time=-1, SpawnRate=40)
  ]
  particleSpawnScript [
   &VC_PaintedVariantsOverride(
    Black=<rgba 0.014 0.014 0.014 1>
    BurntSienna=<rgba 0.16 0.032 0 1>
    Cobalt=<rgba 0.04 0.1227 1 1>
    Crimson=<rgba 0.5271 0.0545 0.0685 1>
    DefaultColor=<rgba 0.2698 2 0.468 1>
    ForestGreen=<rgba 0.077 0.35 0.035 1>
    Gold:LinearColor=<rgba 0.0384 1 0.1096 1>
    Grey=<rgba 0.15 0.15 0.15 1>
    Lime=<rgba 0.3973 0.8 0 1>
    Orange=<rgba 0.8469 0.1301 0.0006 1>
    Pink=<rgba 0.491 0.0648 0.5906 1>
    Purple=<rgba 0.1049 0.0264 0.8796 1>
    Saffron=<rgba 0.65 0.52 0 1>
    SkyBlue=<rgba 0.107 0.4621 0.9301 1>
    TitaniumWhite=<rgba 0.6 0.6 0.6 1>
   ) {enabled=false}
   &VC_InitializeRibbon_v1(
    Color=$Particles.Color
    Custom Offset=true
    Enable Skeletal Position=true
    Facing Axis:ENiagaraOrientationAxis=::ZAxis
    Lifetime=&MergeHelper_FloatSwitch(0=0.6, 1=0.3, 2=0.3, 3=?, Index=$User.GameModeID, Num=::NewEnumerator1)
    Local Vector Offset=&AddVectorToPosition(Position=<-2 0 0>, Vector=$User.DTCarOffset)
    Offset Amount=1
    Ribbon Width=50
   )
   &AddVelocity(Rotation Angle Type:ENiagara_AngleInput=::NewEnumerator1, Rotation Mode=::NewEnumerator2, Velocity=<-1500 0 0>) {module=AddVelocity001}
  ]
  particleUpdateScript [
   &ParticleState()
   &VC_PaintedVariantsOverride(
    Black=<rgba 1 1 1 1>
    BurntSienna=<rgba 30 8.5964 6.2129 1>
    Cobalt=<rgba 1.3901 3.3513 20 1>
    Crimson=<rgba 50 5.8148 5.5923 1>
    DefaultColor=<rgba 6.745 50 11.7001 1>
    ForestGreen=<rgba 10 50 10.5419 1>
    Grey=<rgba 1 1 1 1>
    Lime=<rgba 21.7157 50 4.7866 1>
    Orange=<rgba 50 27.2081 15 1>
    Pink=<rgba 50 15 37.2727 1>
    Purple=<rgba 23.4686 5 50 1>
    R1 Black:LinearColor=<rgba 5 5 5 1>
    R1 BurntSienna:LinearColor=<rgba 50 14.3273 10.3549 1>
    R1 Cobalt:LinearColor=<rgba 2.0852 5.027 30 1>
    R1 Crimson:LinearColor=<rgba 20 4.0802 4 1>
    R1 DefaultColor=&Multiply_LinearColorByFloat(
     Linear Color=<rgba 0.1356 1 0.2346 1>
     Float=&FloatFromCurve(
      FloatCurve=DataInterface() {propertyValues {Curve {
         defaultValue=1
         keys [
          <at 0.65 0.0353 RCIM_Cubic>
          {arriveTangent=-0.2351, interpMode=RCIM_Cubic, leaveTangent=-0.2351, time=0.7094, value=1.3}
          <at 0.8 0 RCIM_Cubic>
         ]
        }}}
      Scale Curve=100
     )
    )
    R1 ForestGreen:LinearColor=<rgba 4 20 4.2168 1>
    R1 Grey:LinearColor=<rgba 5 5 5 1>
    R1 Lime:LinearColor=<rgba 13.0294 30 2.872 1>
    R1 Orange:LinearColor=<rgba 20 10.8833 6 1>
    R1 Pink:LinearColor=<rgba 20 6 14.9091 1>
    R1 Purple:LinearColor=<rgba 9.3875 2 20 1>
    R1 Saffron:LinearColor=<rgba 20 16.7333 6 1>
    R1 SkyBlue:LinearColor=<rgba 6 15.3154 20 1>
    R1 TitaniumWhite:LinearColor=<rgba 20 20 20 1>
    R2 Black:LinearColor=<rgba 1 1 1 0>
    R2 BurntSienna:LinearColor=<rgba 1 0.2865 0.2071 0>
    R2 Cobalt:LinearColor=<rgba 0.2379 0.252 1 0.3>
    R2 Crimson:LinearColor=<rgba 1 0.1163 0.1118 0.15>
    R2 DefaultColor:LinearColor=&Multiply_LinearColorByFloat(
     Linear Color=<rgba 0.1356 1 0.2346 0.5>
     Float=&FloatFromCurve(
      FloatCurve=DataInterface() {propertyValues {Curve {
         defaultValue=1
         keys [
          <at 0.65 0.0353 RCIM_Cubic>
          {arriveTangent=-0.2351, interpMode=RCIM_Cubic, leaveTangent=-0.2351, time=0.7094, value=1.3}
          <at 0.8 0 RCIM_Cubic>
         ]
        }}}
      Scale Curve=0.6
     )
    )
    R2 ForestGreen:LinearColor=<rgba 0 0.5 0.0253 0.1>
    R2 Grey:LinearColor=<rgba 1 1 1 0>
    R2 Lime:LinearColor=<rgba 0.2623 0.6038 0.0578 0.2>
    R2 Orange:LinearColor=<rgba 0.6724 0.2346 0 0.2>
    R2 Pink:LinearColor=<rgba 0.5052 0.0605 0.3435 0.2>
    R2 Purple:LinearColor=<rgba 0.2789 0 0.6795 0.2>
    R2 Saffron:LinearColor=<rgba 0.9 0.69 0 0.2>
    R2 SkyBlue:LinearColor=<rgba 0 0.357 0.5365 0.2>
    R2 TitaniumWhite:LinearColor=<rgba 0.75 0.75 0.75 0.15>
    Saffron=<rgba 50 41.8333 15 1>
    SkyBlue=<rgba 6 15.3154 20 1>
    TitaniumWhite=<rgba 20 20 20 1>
    Use R1 Painted Overrides=true
    Use R2 Painted Overrides=true
   ) {module=VC_PaintedVariantsOverride001}
   &InheritSourceMovement(
    Applied Position Scale=&VectorFromFloat(Value=&FloatFromCurve(FloatCurve=DataInterface() {propertyValues {Curve {defaultValue=1, keys [{interpMode=RCIM_Cubic, tangentMode=RCTM_User, time=0, value=1}, {tangentMode=RCTM_User, time=1, value=0}]}}}))
    ApplyForce=false
    ApplyPosition=true
    ApplyPositionFalloff=false
   )
   &ScaleRibbonWidth(Ribbon Width Scale=&FloatFromCurve(
     FloatCurve=DataInterface() {propertyValues {Curve {keys [
         <at 0 1.5522 RCIM_Cubic>
         {arriveTangent=2.0161, interpMode=RCIM_Cubic, leaveTangent=2.0161, time=0.1571, value=-0.5}
         {arriveTangent=0.673, interpMode=RCIM_Cubic, leaveTangent=0.673, time=0.6227, value=2.8076}
         <at 0.9 0>
        ]}}}
     CurveIndex=&Multiply_Float(A=$Particles.NormalizedAge, B=&RandomRangeFloat(Minimum=0.1, Maximum=3))
     Scale Curve=0.6
    ))
   /SetVariables(
    Particles.WidthGlow:NiagaraFloat=&FloatFromCurve(
     FloatCurve=DataInterface() {propertyValues {Curve {
        defaultValue=1
        keys [
         {arriveTangent=4.4334, interpMode=RCIM_Cubic, time=0.0145, value=0.0922}
         {arriveTangent=0.0544, interpMode=RCIM_Cubic, leaveTangent=0.0544, tangentMode=RCTM_User, time=0.4056, value=1}
         {arriveTangent=-0.0405, interpMode=RCIM_Cubic, leaveTangent=-0.0405, tangentMode=RCTM_User, time=0.7009, value=1}
         {arriveTangent=-8.0089, interpMode=RCIM_Cubic, leaveTangent=-8.0089, tangentMode=RCTM_User, time=1, value=0}
        ]
       }}}
     Scale Curve=50
    )
    Particles.WidthThunder:NiagaraFloat=&Multiply_Float(A=$Particles.RibbonWidth, B=2.5)
   ) {isSetParameters=true}
   &ScaleVelocity(
    Coordinate Space:ENiagaraCoordinateSpace=::Simulation
    Velocity Scale=&VectorFromCurve(
     VectorCurve=DataInterface() {propertyValues {XCurve {keys [<at 0 3>,<at 3 0.1>]}, YCurve {keys [<at 0 3>,<at 3 0.1>]}, ZCurve {keys [<at 0 2>,<at 2.5 0.5>]}}}
     Scale Curve=<0.5 0.5 1>
    )
   )
   &JitterPosition(Jitter Amount=2, Jitter Delay=0.5)
   &SolveForcesAndVelocity(Clamp Velocity=false, Limit Acceleration=false)
   &VC_UpdateRibbon_v1(Enable Skeletal Position=true)
  ]
  rendererClasses [&NiagaraRibbonRendererProperties]
  renderers [
   {rendererClass=&NiagaraRibbonRendererProperties}
   {rendererClass=&NiagaraRibbonRendererProperties, rendererIndex=1}
   {rendererClass=&NiagaraRibbonRendererProperties, rendererIndex=2}
  ]
 }
 NE_VC_Boost_Ribbon_Trail001 {
  emitterUpdateScript [
   &EmitterState(Life Cycle Mode=::NewEnumerator1, Loop Duration=1, UseLoopDelay=false)
   &VC_FilteredSocketCount()
   &VC_RateQualityHandler(Epic Spawn Rate=30, High Spawn Rate=30, Low Spawn Rate=30, Use Custom Rate?=true)
   &VC_SpawnRibbon_v1(Number of Ribbons=$Emitter.PortCount, SpawnRate=$Emitter.Spawn Rate)
  ]
  particleSpawnScript [
   &VC_PaintedVariantsOverride(DefaultColor=&ColorFromCurve(ColorCurve=DataInterface() {propertyValues {
       AlphaCurve {keys [<at 0 1 RCIM_Cubic>,<at 0.453 0 RCIM_Cubic>]}
       BlueCurve {keys [<at 0 0.4 RCIM_Cubic>,<at 0.383 1 RCIM_Cubic>]}
       GreenCurve {keys [<at 0 1 RCIM_Cubic>,<at 0.383 0.0896 RCIM_Cubic>]}
       RedCurve {keys [<at 0 0.5 RCIM_Cubic>,<at 0.383 0.0629 RCIM_Cubic>]}
      }})) {enabled=false}
   &VC_InitializeRibbon_v1(
    Color=$Particles.Color
    Custom Offset=true
    Enable Skeletal Position=true
    Lifetime=&MergeHelper_FloatSwitch(0=0.9, 1=0.5, 2=0.5, 3=?, Index=$User.GameModeID, Num=::NewEnumerator1)
    Local Vector Offset=&AddVectorToPosition(Position=<-55 0 0>, Vector=$User.DTCarOffset)
    Offset Amount=1
    Ribbon Width=40
   )
   &AddVelocity(Velocity=&VC_Mask_Vectors(
     FrontEnd Value:Vector3f=<-1199 0 0>
     blsFrontEndPreview:NiagaraBool=true
     SpeedMin Value:Vector3f=<-1000 0 0>
     SpeedMax Value:Vector3f=<-1000 0 0>
     Speed Alpha:NiagaraFloat=$System.Speed_Alpha
    )) {module=AddVelocity001}
  ]
  particleUpdateScript [
   &ParticleState()
   &VC_PaintedVariantsOverride(
    Black=&ColorFromCurve(ColorCurve=DataInterface() {propertyValues {
       AlphaCurve {keys [<at 0 1 RCIM_Cubic>]}
       BlueCurve {keys [<at 0 0.05 RCIM_Cubic>,<at 0.2373 0.02>]}
       GreenCurve {keys [<at 0 0.0474 RCIM_Cubic>,<at 0.2373 0.02>]}
       RedCurve {keys [<at 0 0.0468 RCIM_Cubic>,<at 0.2373 0.02>]}
      }})
    BurntSienna=&ColorFromCurve(ColorCurve=DataInterface() {propertyValues {
       AlphaCurve {keys [<at 0 1 RCIM_Cubic>]}
       BlueCurve {keys [<at 0 0.003 RCIM_Cubic>,<at 0.5249 0.0024>]}
       GreenCurve {keys [<at 0 0.0424 RCIM_Cubic>,<at 0.5249 0.0091>]}
       RedCurve {keys [<at 0 0.2 RCIM_Cubic>,<at 0.5249 0.0704>]}
      }})
    Cobalt=&ColorFromCurve(ColorCurve=DataInterface() {propertyValues {
       AlphaCurve {keys [<at 0 1 RCIM_Cubic>]}
       BlueCurve {keys [<at 0 1 RCIM_Cubic>,<at 0.1831 0.4178>,<at 0.4094 0.18>]}
       GreenCurve {keys [<at 0 0.7682 RCIM_Cubic>,<at 0.1831 0.2403>,<at 0.4094 0.0247>]}
       RedCurve {keys [<at 0 0.5271 RCIM_Cubic>,<at 0.1831 0.1612>,<at 0.4094 0.0117>]}
      }})
    Crimson=&ColorFromCurve(ColorCurve=DataInterface() {propertyValues {
       AlphaCurve {keys [<at 0 1 RCIM_Cubic>]}
       BlueCurve {keys [<at 0 0.3231 RCIM_Cubic>,<at 0.1831 0.0702>,<at 0.4363 0.004>]}
       GreenCurve {keys [<at 0 0.3231 RCIM_Cubic>,<at 0.1831 0.0702>,<at 0.4363 0.004>]}
       RedCurve {keys [<at 0 0.9911 RCIM_Cubic>,<at 0.1831 0.4>,<at 0.4363 0.0467>]}
      }})
    DefaultColor=&ColorFromCurve(ColorCurve=DataInterface() {propertyValues {
       AlphaCurve {keys [<at 0 1 RCIM_Cubic>]}
       BlueCurve {keys [<at 0 0.5 RCIM_Cubic>, {arriveTangent=-0.9461, interpMode=RCIM_Cubic, leaveTangent=-0.9461, time=0.232, value=0.1335}, <at 0.5073 0.02>]}
       GreenCurve {keys [<at 0 0.5 RCIM_Cubic>, {arriveTangent=-0.7884, interpMode=RCIM_Cubic, leaveTangent=-0.7884, time=0.232, value=0.4}, <at 0.5073 0.1>]}
       RedCurve {keys [<at 0 0.2575 RCIM_Cubic>, {arriveTangent=-0.468, interpMode=RCIM_Cubic, leaveTangent=-0.468, time=0.232, value=0.034}, <at 0.5073 0.02>]}
      }})
    ForestGreen=&ColorFromCurve(ColorCurve=DataInterface() {propertyValues {
       AlphaCurve {keys [<at 0 1 RCIM_Cubic>]}
       BlueCurve {keys [<at 0 0.25 RCIM_Cubic>,<at 0.1831 0.0251>,<at 0.4363 0.0052>]}
       GreenCurve {keys [<at 0 1 RCIM_Cubic>,<at 0.1831 0.2>,<at 0.4363 0.0395>]}
       RedCurve {keys [<at 0 0.25 RCIM_Cubic>,<at 0.1831 0.0251>,<at 0.4363 0.0048>]}
      }})
    Grey=&ColorFromCurve(ColorCurve=DataInterface() {propertyValues {
       AlphaCurve {keys [<at 0 1 RCIM_Cubic>]}
       BlueCurve {keys [<at 0 0.9911 RCIM_Cubic>,<at 0.1831 0.1946>,<at 0.4363 0.02>]}
       GreenCurve {keys [<at 0 0.9911 RCIM_Cubic>,<at 0.1831 0.1946>,<at 0.4363 0.0181>]}
       RedCurve {keys [<at 0 0.9911 RCIM_Cubic>,<at 0.1831 0.1946>,<at 0.4363 0.0181>]}
      }})
    Lime=&ColorFromCurve(ColorCurve=DataInterface() {propertyValues {
       AlphaCurve {keys [<at 0 1 RCIM_Cubic>]}
       BlueCurve {keys [<at 0 0.3968 RCIM_Cubic>,<at 0.1563 0.0953>,<at 0.3979 0.0066>]}
       GreenCurve {keys [<at 0 1 RCIM_Cubic>,<at 0.1563 0.9131>,<at 0.3979 0.08>]}
       RedCurve {keys [<at 0 1 RCIM_Cubic>,<at 0.1563 0.6038>,<at 0.3979 0.034>]}
      }})
    Orange=&ColorFromCurve(ColorCurve=DataInterface() {propertyValues {
       AlphaCurve {keys [<at 0 1 RCIM_Cubic>]}
       BlueCurve {keys [<at 0 0.3515 RCIM_Cubic>,<at 0.1927 0.0437>,<at 0.3979 0.007>]}
       GreenCurve {keys [<at 0 0.7779 RCIM_Cubic>,<at 0.1927 0.2086>,<at 0.3979 0.0252>]}
       RedCurve {keys [<at 0 0.9911 RCIM_Cubic>,<at 0.1927 0.5457>,<at 0.3979 0.1022>]}
      }})
    Pink=&ColorFromCurve(ColorCurve=DataInterface() {propertyValues {
       AlphaCurve {keys [<at 0 1 RCIM_Cubic>]}
       BlueCurve {keys [<at 0 0.965 RCIM_Cubic>,<at 0.1927 0.3968>,<at 0.3979 0.0307>]}
       GreenCurve {keys [<at 0 0.44 RCIM_Cubic>,<at 0.1927 0.1274>,<at 0.3979 0.008>]}
       RedCurve {keys [<at 0 1 RCIM_Cubic>,<at 0.1927 0.4969>,<at 0.3979 0.0482>]}
      }})
    Purple=&ColorFromCurve(ColorCurve=DataInterface() {propertyValues {
       AlphaCurve {keys [<at 0 1 RCIM_Cubic>]}
       BlueCurve {keys [<at 0 1 RCIM_Cubic>,<at 0.093 0.5>,<at 0.3979 0.0595>]}
       GreenCurve {keys [<at 0 0.1 RCIM_Cubic>,<at 0.093 0.1639>,<at 0.3979 0.0048>]}
       RedCurve {keys [<at 0 0.829 RCIM_Cubic>,<at 0.093 0.4079>,<at 0.3979 0.0252>]}
      }})
    R1 Black:LinearColor=&ColorFromCurve(ColorCurve=DataInterface() {propertyValues {
       AlphaCurve {keys [<at 0.2009 1 RCIM_Cubic>,<at 1 0 RCIM_Cubic>]}
       BlueCurve {keys [<at 0 0 RCIM_Cubic>]}
       GreenCurve {keys [<at 0 0 RCIM_Cubic>]}
       RedCurve {keys [<at 0 0 RCIM_Cubic>]}
      }})
    R1 BurntSienna:LinearColor=&ColorFromCurve(ColorCurve=DataInterface() {propertyValues {
       AlphaCurve {keys [<at 0.2009 1 RCIM_Cubic>,<at 1 0 RCIM_Cubic>]}
       BlueCurve {keys [<at 0 0 RCIM_Cubic>]}
       GreenCurve {keys [<at 0 0 RCIM_Cubic>]}
       RedCurve {keys [<at 0 0 RCIM_Cubic>]}
      }})
    R1 Cobalt:LinearColor=&ColorFromCurve(ColorCurve=DataInterface() {propertyValues {
       AlphaCurve {keys [<at 0.2009 1 RCIM_Cubic>,<at 1 0 RCIM_Cubic>]}
       BlueCurve {keys [<at 0 0 RCIM_Cubic>]}
       GreenCurve {keys [<at 0 0 RCIM_Cubic>]}
       RedCurve {keys [<at 0 0 RCIM_Cubic>]}
      }})
    R1 Crimson:LinearColor=&ColorFromCurve(ColorCurve=DataInterface() {propertyValues {
       AlphaCurve {keys [<at 0.2009 1 RCIM_Cubic>,<at 1 0 RCIM_Cubic>]}
       BlueCurve {keys [<at 0 0 RCIM_Cubic>]}
       GreenCurve {keys [<at 0 0 RCIM_Cubic>]}
       RedCurve {keys [<at 0 0 RCIM_Cubic>]}
      }})
    R1 DefaultColor=&ColorFromCurve(ColorCurve=DataInterface() {propertyValues {
       AlphaCurve {keys [<at 0.2009 1 RCIM_Cubic>,<at 1 0 RCIM_Cubic>]}
       BlueCurve {keys [<at 0 0 RCIM_Cubic>]}
       GreenCurve {keys [<at 0 0 RCIM_Cubic>]}
       RedCurve {keys [<at 0 0 RCIM_Cubic>]}
      }})
    R1 ForestGreen:LinearColor=&ColorFromCurve(ColorCurve=DataInterface() {propertyValues {
       AlphaCurve {keys [<at 0.2009 1 RCIM_Cubic>,<at 1 0 RCIM_Cubic>]}
       BlueCurve {keys [<at 0 0 RCIM_Cubic>]}
       GreenCurve {keys [<at 0 0 RCIM_Cubic>]}
       RedCurve {keys [<at 0 0 RCIM_Cubic>]}
      }})
    R1 Grey:LinearColor=&ColorFromCurve(ColorCurve=DataInterface() {propertyValues {
       AlphaCurve {keys [<at 0.2009 1 RCIM_Cubic>,<at 1 0 RCIM_Cubic>]}
       BlueCurve {keys [<at 0 0 RCIM_Cubic>]}
       GreenCurve {keys [<at 0 0 RCIM_Cubic>]}
       RedCurve {keys [<at 0 0 RCIM_Cubic>]}
      }})
    R1 Lime:LinearColor=&ColorFromCurve(ColorCurve=DataInterface() {propertyValues {
       AlphaCurve {keys [<at 0.2009 1 RCIM_Cubic>,<at 1 0 RCIM_Cubic>]}
       BlueCurve {keys [<at 0 0 RCIM_Cubic>]}
       GreenCurve {keys [<at 0 0 RCIM_Cubic>]}
       RedCurve {keys [<at 0 0 RCIM_Cubic>]}
      }})
    R1 Orange:LinearColor=&ColorFromCurve(ColorCurve=DataInterface() {propertyValues {
       AlphaCurve {keys [<at 0.2009 1 RCIM_Cubic>,<at 1 0 RCIM_Cubic>]}
       BlueCurve {keys [<at 0 0 RCIM_Cubic>]}
       GreenCurve {keys [<at 0 0 RCIM_Cubic>]}
       RedCurve {keys [<at 0 0 RCIM_Cubic>]}
      }})
    R1 Pink:LinearColor=&ColorFromCurve(ColorCurve=DataInterface() {propertyValues {
       AlphaCurve {keys [<at 0.2009 1 RCIM_Cubic>,<at 1 0 RCIM_Cubic>]}
       BlueCurve {keys [<at 0 0 RCIM_Cubic>]}
       GreenCurve {keys [<at 0 0 RCIM_Cubic>]}
       RedCurve {keys [<at 0 0 RCIM_Cubic>]}
      }})
    R1 Purple:LinearColor=&ColorFromCurve(ColorCurve=DataInterface() {propertyValues {
       AlphaCurve {keys [<at 0.2009 1 RCIM_Cubic>,<at 1 0 RCIM_Cubic>]}
       BlueCurve {keys [<at 0 0 RCIM_Cubic>]}
       GreenCurve {keys [<at 0 0 RCIM_Cubic>]}
       RedCurve {keys [<at 0 0 RCIM_Cubic>]}
      }})
    R1 Saffron:LinearColor=&ColorFromCurve(ColorCurve=DataInterface() {propertyValues {
       AlphaCurve {keys [<at 0.2009 1 RCIM_Cubic>,<at 1 0 RCIM_Cubic>]}
       BlueCurve {keys [<at 0 0 RCIM_Cubic>]}
       GreenCurve {keys [<at 0 0 RCIM_Cubic>]}
       RedCurve {keys [<at 0 0 RCIM_Cubic>]}
      }})
    R1 SkyBlue:LinearColor=&ColorFromCurve(ColorCurve=DataInterface() {propertyValues {
       AlphaCurve {keys [<at 0.2009 1 RCIM_Cubic>,<at 1 0 RCIM_Cubic>]}
       BlueCurve {keys [<at 0 0 RCIM_Cubic>]}
       GreenCurve {keys [<at 0 0 RCIM_Cubic>]}
       RedCurve {keys [<at 0 0 RCIM_Cubic>]}
      }})
    R1 TitaniumWhite:LinearColor=&ColorFromCurve(ColorCurve=DataInterface() {propertyValues {
       AlphaCurve {keys [<at 0.2009 1 RCIM_Cubic>,<at 1 0 RCIM_Cubic>]}
       BlueCurve {keys [<at 0 0 RCIM_Cubic>]}
       GreenCurve {keys [<at 0 0 RCIM_Cubic>]}
       RedCurve {keys [<at 0 0 RCIM_Cubic>]}
      }})
    R2 Black:LinearColor=&ColorFromCurve(ColorCurve=DataInterface() {propertyValues {
       AlphaCurve {keys [<at 0 1 RCIM_Cubic>,<at 0.7732 0 RCIM_Cubic>]}
       BlueCurve {keys [<at 0 0 RCIM_Cubic>]}
       GreenCurve {keys [<at 0 0 RCIM_Cubic>]}
       RedCurve {keys [<at 0 0 RCIM_Cubic>]}
      }})
    R2 BurntSienna:LinearColor=&ColorFromCurve(ColorCurve=DataInterface() {propertyValues {
       AlphaCurve {keys [<at 0 1 RCIM_Cubic>,<at 0.7732 0 RCIM_Cubic>]}
       BlueCurve {keys [<at 0 0 RCIM_Cubic>]}
       GreenCurve {keys [<at 0 0 RCIM_Cubic>]}
       RedCurve {keys [<at 0 0 RCIM_Cubic>]}
      }})
    R2 Cobalt:LinearColor=&ColorFromCurve(ColorCurve=DataInterface() {propertyValues {
       AlphaCurve {keys [<at 0 1 RCIM_Cubic>,<at 0.7732 0 RCIM_Cubic>]}
       BlueCurve {keys [<at 0 2 RCIM_Cubic>]}
       GreenCurve {keys [<at 0 2 RCIM_Cubic>]}
       RedCurve {keys [<at 0 2 RCIM_Cubic>]}
      }})
    R2 Crimson:LinearColor=&ColorFromCurve(ColorCurve=DataInterface() {propertyValues {
       AlphaCurve {keys [<at 0 1 RCIM_Cubic>,<at 0.7732 0 RCIM_Cubic>]}
       BlueCurve {keys [<at 0 2 RCIM_Cubic>]}
       GreenCurve {keys [<at 0 2 RCIM_Cubic>]}
       RedCurve {keys [<at 0 2 RCIM_Cubic>]}
      }})
    R2 DefaultColor:LinearColor=&ColorFromCurve(ColorCurve=DataInterface() {propertyValues {
       AlphaCurve {keys [<at 0 1 RCIM_Cubic>,<at 0.7732 0 RCIM_Cubic>]}
       BlueCurve {keys [<at 0 2 RCIM_Cubic>]}
       GreenCurve {keys [<at 0 2 RCIM_Cubic>]}
       RedCurve {keys [<at 0 2 RCIM_Cubic>]}
      }})
    R2 ForestGreen:LinearColor=&ColorFromCurve(ColorCurve=DataInterface() {propertyValues {
       AlphaCurve {keys [<at 0 1 RCIM_Cubic>,<at 0.7732 0 RCIM_Cubic>]}
       BlueCurve {keys [<at 0 2 RCIM_Cubic>]}
       GreenCurve {keys [<at 0 2 RCIM_Cubic>]}
       RedCurve {keys [<at 0 2 RCIM_Cubic>]}
      }})
    R2 Grey:LinearColor=&ColorFromCurve(ColorCurve=DataInterface() {propertyValues {
       AlphaCurve {keys [<at 0 1 RCIM_Cubic>,<at 0.7732 0 RCIM_Cubic>]}
       BlueCurve {keys [<at 0 2 RCIM_Cubic>]}
       GreenCurve {keys [<at 0 2 RCIM_Cubic>]}
       RedCurve {keys [<at 0 2 RCIM_Cubic>]}
      }})
    R2 Lime:LinearColor=&ColorFromCurve(ColorCurve=DataInterface() {propertyValues {
       AlphaCurve {keys [<at 0 1 RCIM_Cubic>,<at 0.7732 0 RCIM_Cubic>]}
       BlueCurve {keys [<at 0 2 RCIM_Cubic>]}
       GreenCurve {keys [<at 0 2 RCIM_Cubic>]}
       RedCurve {keys [<at 0 2 RCIM_Cubic>]}
      }})
    R2 Orange:LinearColor=&ColorFromCurve(ColorCurve=DataInterface() {propertyValues {
       AlphaCurve {keys [<at 0 1 RCIM_Cubic>,<at 0.7732 0 RCIM_Cubic>]}
       BlueCurve {keys [<at 0 2 RCIM_Cubic>]}
       GreenCurve {keys [<at 0 2 RCIM_Cubic>]}
       RedCurve {keys [<at 0 2 RCIM_Cubic>]}
      }})
    R2 Pink:LinearColor=&ColorFromCurve(ColorCurve=DataInterface() {propertyValues {
       AlphaCurve {keys [<at 0 1 RCIM_Cubic>,<at 0.7732 0 RCIM_Cubic>]}
       BlueCurve {keys [<at 0 2 RCIM_Cubic>]}
       GreenCurve {keys [<at 0 2 RCIM_Cubic>]}
       RedCurve {keys [<at 0 2 RCIM_Cubic>]}
      }})
    R2 Purple:LinearColor=&ColorFromCurve(ColorCurve=DataInterface() {propertyValues {
       AlphaCurve {keys [<at 0 1 RCIM_Cubic>,<at 0.7732 0 RCIM_Cubic>]}
       BlueCurve {keys [<at 0 2 RCIM_Cubic>]}
       GreenCurve {keys [<at 0 2 RCIM_Cubic>]}
       RedCurve {keys [<at 0 2 RCIM_Cubic>]}
      }})
    R2 Saffron:LinearColor=&ColorFromCurve(ColorCurve=DataInterface() {propertyValues {
       AlphaCurve {keys [<at 0 1 RCIM_Cubic>,<at 0.7732 0 RCIM_Cubic>]}
       BlueCurve {keys [<at 0 2 RCIM_Cubic>]}
       GreenCurve {keys [<at 0 2 RCIM_Cubic>]}
       RedCurve {keys [<at 0 2 RCIM_Cubic>]}
      }})
    R2 SkyBlue:LinearColor=&ColorFromCurve(ColorCurve=DataInterface() {propertyValues {
       AlphaCurve {keys [<at 0 1 RCIM_Cubic>,<at 0.7732 0 RCIM_Cubic>]}
       BlueCurve {keys [<at 0 2 RCIM_Cubic>]}
       GreenCurve {keys [<at 0 2 RCIM_Cubic>]}
       RedCurve {keys [<at 0 2 RCIM_Cubic>]}
      }})
    R2 TitaniumWhite:LinearColor=&ColorFromCurve(ColorCurve=DataInterface() {propertyValues {
       AlphaCurve {keys [<at 0 1 RCIM_Cubic>,<at 0.7732 0 RCIM_Cubic>]}
       BlueCurve {keys [<at 0 2 RCIM_Cubic>]}
       GreenCurve {keys [<at 0 2 RCIM_Cubic>]}
       RedCurve {keys [<at 0 2 RCIM_Cubic>]}
      }})
    Saffron=&ColorFromCurve(ColorCurve=DataInterface() {propertyValues {
       AlphaCurve {keys [<at 0 1 RCIM_Cubic>]}
       BlueCurve {keys [<at 0 0.4179 RCIM_Cubic>,<at 0.1927 0.1095>,<at 0.3979 0>]}
       GreenCurve {keys [<at 0 1 RCIM_Cubic>,<at 0.1927 0.8469>,<at 0.3979 0.0972>]}
       RedCurve {keys [<at 0 1 RCIM_Cubic>,<at 0.1927 0.855>,<at 0.3979 0.1>]}
      }})
    SkyBlue=&ColorFromCurve(ColorCurve=DataInterface() {propertyValues {
       AlphaCurve {keys [<at 0 1 RCIM_Cubic>]}
       BlueCurve {keys [<at 0 0.9734 RCIM_Cubic>,<at 0.1486 0.9047>,<at 0.3979 0.0802>]}
       GreenCurve {keys [<at 0 1 RCIM_Cubic>,<at 0.1486 0.7157>,<at 0.3979 0.0409>]}
       RedCurve {keys [<at 0 0.4621 RCIM_Cubic>,<at 0.1486 0.1441>,<at 0.3979 0.0065>]}
      }})
    TitaniumWhite=&ColorFromCurve(ColorCurve=DataInterface() {propertyValues {
       AlphaCurve {keys [<at 0 1 RCIM_Cubic>]}
       BlueCurve {keys [<at 0 1.5 RCIM_Cubic>,<at 0.2023 0.3652>,<at 0.3979 0.12>]}
       GreenCurve {keys [<at 0 1.5 RCIM_Cubic>,<at 0.2023 0.3631>,<at 0.3979 0.1156>]}
       RedCurve {keys [<at 0 1.5 RCIM_Cubic>,<at 0.2023 0.3631>,<at 0.3979 0.1156>]}
      }})
    Use R1 Painted Overrides=true
    Use R2 Painted Overrides=true
   ) {module=VC_PaintedVariantsOverride001}
   &DynamicMaterialParameters(
    Index 1 Param 1=&FloatFromCurve(
     FloatCurve=DataInterface() {propertyValues {Curve {keys [
         {arriveTangent=-6.0653, interpMode=RCIM_Cubic, leaveTangent=-6.0653, tangentMode=RCTM_User, time=0, value=-6}
         {arriveTangent=-0.0284, interpMode=RCIM_Cubic, leaveTangent=-0.0284, tangentMode=RCTM_User, time=1, value=4}
        ]}}}
     Scale Curve=0.3
    )
    Index 1 Param 2=1
    Module.Index1Float0Write=true
    Module.Index1Float1Write=true
    Module.Index1Float2Write=true
    Module.Index1Float3Write=true
    Write Parameter Index 0=::NewEnumerator0
    Write Parameter Index 1=::NewEnumerator1
   )
   /SetVariables(
    Particles.Dissolves:Vector4f=&MakeVector4(
     X=&FloatFromCurve(
      FloatCurve=DataInterface() {propertyValues {Curve {keys [
          {arriveTangent=-6.0653, interpMode=RCIM_Cubic, leaveTangent=-6.0653, tangentMode=RCTM_User, time=0, value=-4}
          {arriveTangent=-0.0284, interpMode=RCIM_Cubic, leaveTangent=-0.0284, tangentMode=RCTM_User, time=1, value=4}
         ]}}}
      Scale Curve=0
     )
     Y=0
     Z=0
     W=0
    )
    Particles.ScaleRibbon:NiagaraFloat=&FloatFromCurve(
     FloatCurve=DataInterface() {propertyValues {Curve {defaultValue=1, keys [{tangentMode=RCTM_User, time=0, value=0.7}, <at 0.9 0.4 RCIM_Constant>, <at 0.91 0 RCIM_Cubic>]}}}
     Scale Curve=38
    )
   ) {isSetParameters=true}
   /SetVariables(Particles.Position=&AddVectorToPosition(
     Position=$Particles.Position
     Vector=&TransformVector(
      Vector=&MakeVector(
       X=0
       Y=0
       Z=&Sine(
        Normalized Angle=&Add_Float(
         A=&Multiply_Float(A=$Particles.NormalizedAge, B=-1)
         B=&Multiply_Float(A=$Emitter.LoopedAge, B=2)
        )
        Period=1
        Scale=&FloatFromCurve(FloatCurve=DataInterface() {propertyValues {Curve {keys [<at -0.4 0>,<at 0.35 0.5>]}}}, Scale Curve=2.5)
        Bias=0
       )
      )
      Destination Space=::Simulation
      Apply Scale=true
     )
    )) {isSetParameters=true}
   &InheritSourceMovement(
    Applied Position Scale=&VectorFromFloat(Value=&FloatFromCurve(FloatCurve=DataInterface() {propertyValues {Curve {keys [
          {interpMode=RCIM_Cubic, tangentMode=RCTM_Break, time=0, value=1}
          {arriveTangent=-2.9025, interpMode=RCIM_Cubic, tangentMode=RCTM_Break, time=1, value=0}
         ]}}}))
    ApplyForce=false
    ApplyPosition=true
    ApplyPositionFalloff=false
   )
   &SolveForcesAndVelocity(Clamp Velocity=false, Limit Acceleration=false)
   &VC_UpdateRibbon_v1(Enable Skeletal Position=true)
  ]
  rendererClasses [&NiagaraRibbonRendererProperties]
  renderers [
   {rendererClass=&NiagaraRibbonRendererProperties}
   {rendererClass=&NiagaraRibbonRendererProperties, rendererIndex=1}
   {rendererClass=&NiagaraRibbonRendererProperties, rendererIndex=2}
   {rendererClass=&NiagaraRibbonRendererProperties, rendererIndex=3}
   {rendererClass=&NiagaraRibbonRendererProperties, rendererIndex=4}
   {rendererClass=&NiagaraRibbonRendererProperties, rendererIndex=5}
  ]
 }
 NE_VC_Boost_Sprite_Parent002 {
  emitterUpdateScript [
   &EmitterState(Inactive Response=::NewEnumerator1, Life Cycle Mode=::NewEnumerator1, Loop Duration=0.1, UseLoopDelay=false)
   &SpawnBurst_Instantaneous(Spawn Count=4, Use Spawn Probability=false, UseLoopCountLimit=false)
   &VC_FilteredSocketCount()
   &SpawnRate(SpawnRate=60, Use Spawn Probability=false) {enabled=false}
  ]
  particleSpawnScript [
   &VC_PaintedVariantsOverride(
    Black=<rgba 0.0098 0.01 0.0088 1>
    BurntSienna=<rgba 0.05 0.0181 0.0075 1>
    Cobalt=<rgba 0.06 0.0626 0.2 1>
    Crimson=<rgba 0.06 0.0265 0.0273 1>
    DefaultColor=<rgba 0.0394 0.08 0.0683 1>
    ForestGreen=<rgba 0.032 0.08 0.0361 1>
    Grey=<rgba 0.0497 0.0497 0.0497 1>
    Lime=<rgba 0.0723 0.0908 0.0356 1>
    Orange=<rgba 0.0802 0.0648 0.0319 1>
    Pink=<rgba 0.0802 0.0561 0.0802 1>
    Purple=<rgba 0.0762 0.0513 0.0802 1>
    Saffron=<rgba 0.0802 0.0802 0.0395 1>
    SkyBlue=<rgba 0.0482 0.0802 0.0802 1>
    TitaniumWhite=<rgba 0.0497 0.0497 0.0497 1>
   )
   &InitializeParticle(
    Color=$Particles.Color
    Lifetime=0.15
    Mesh Scale Max=&Multiply_VectorByFloat(Vector=<1 1 2>, Float=2)
    Mesh Scale Min=&Multiply_VectorByFloat(Vector=<1 1 1.5>, Float=2)
    Mesh Scale Mode=::NewEnumerator2
    Sprite Rotation Mode=::NewEnumerator2
    UsePositionOffset=false
    Write Material Random=true
   )
   &PartitionParticles(Number Of Partitions=$Emitter.PortCount)
   &SkeletalMeshLocation(
    Bone / Socket Index=$Particles.PartitionParticles.Partition
    Bone / Socket Sampling Mode=::NewEnumerator5
    MaskSpawnGroup=false
    Mesh Sampling Type=::NewEnumerator6
    Sampled Position Offset=&AddVectorToPosition(
     Position=&MergeHelper_VectorSwitch(0=<-55 0 0>, 1=<-55 0 0>, 2=<-55 0 0>, 3=?, Index=$User.GameModeID, Num=::NewEnumerator1)
     Vector=$User.DTCarOffset
    )
    Skeletal Mesh=$System.SkeletalMesh
   )
   &InitialMeshOrientation(
    Mesh Axis To Orient=<0 1 0>
    Mesh Orientation Mode=::NewEnumerator2
    Orientation Vector=<0 0 0>
    Rotation=&RandomRangeVector(Minimum=<0 -1 0>, Maximum=<0 1 0>)
    Use Rotation Vector=true
   ) {module=InitialMeshOrientation001}
   &AddVelocity(Velocity=<-1500 0 0>) {enabled=false}
  ]
  particleUpdateScript [
   &ParticleState()
   &SolveForcesAndVelocity(Clamp Velocity=false, Limit Acceleration=false) {enabled=false}
  ]
  rendererClasses [&NiagaraMeshRendererProperties, &NiagaraSpriteRendererProperties]
  renderers [
   {rendererClass=&NiagaraMeshRendererProperties, rendererIndex=0}
   {rendererClass=&NiagaraSpriteRendererProperties, rendererIndex=1}
  ]
 }
}
summary {
 systemName=NS_Boost_Hydroburst
 emitters [
  {emitterName=NE_VC_Boost_Ribbon_Trail001, rendererClasses [&NiagaraRibbonRendererProperties]}
  {
   emitterName=NE_VC_Boost_Sprite_Parent002
   rendererClasses [&NiagaraMeshRendererProperties, &NiagaraSpriteRendererProperties]
  }
  {emitterName=NE_VC_Boost_Ribbon_Parent003, rendererClasses [&NiagaraRibbonRendererProperties]}
 ]
 userVariables [
  User.bIsFrontendPreview {type=NiagaraBool}
  User.bIsPainted {type=NiagaraBool}
  User.ColorVariantID {type=NiagaraInt32}
  User.DTCarOffset {type=Vector3f, description="", defaultValue {struct=Vector3f, value=<0 0 0>}}
  User.GameModeID {type=NiagaraInt32}
  User.PaintedColor {type=LinearColor, description="", defaultValue {struct=LinearColor, value=<rgba 1 1 1 1>}}
  User.PaintedScalar {type=NiagaraFloat, defaultValue=1, description=""}
  User.PortCount {type=NiagaraInt32, defaultValue=2}
 ]
}
